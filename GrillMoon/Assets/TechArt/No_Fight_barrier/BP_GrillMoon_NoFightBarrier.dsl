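; GrillMoon no-fight barrier Blueprint: a zone actor that signals material effects when the local player crosses its boundary.
; Begin and end overlap on Overlap_Volume both check for a PlayerPawn_Athena controlled by player controller 0.
; Either crossing writes current game time to Post_time in MPC_GrillMoon, a timestamp for materials to read, and prints a debug string.
; Clear_Zone destroys the actor once the event no longer needs it; no combat-blocking logic appears here.
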(event Custom|Clear_Zone
  ;[0.1823 0.02923 0.1633 0.5] Clear the zone once no longer needed in event
  (Actor|DestroyActor))

(event OnComponentBeginOverlap(Overlap_Volume)
  (params (OverlappedComponent :type Primitive%20Component%20Object%20Reference) (OtherActor :type Actor%20Object%20Reference) (OtherComp :type Primitive%20Component%20Object%20Reference) (OtherBodyIndex :type Integer) (bFromSweep :type Boolean) (SweepResult :type Hit%20Result%20Structure%20%28by%20ref%29))
  (bind _asplayer_pawn_athena (Utilities|Casting|CastToPlayerPawn_Athena OtherActor))
  (if (== (Pawn|GetController _asplayer_pawn_athena) (Game|GetPlayerController 0))
    (Rendering|Material|SetScalarParameterValue "/GrillMoon_Countdown/Assets/FX/MPC_GrillMoon.MPC_GrillMoon" "Post_time" (Utilities|Time|GetGameTimeinSeconds) :!class /Script/Engine.KismetMaterialLibrary)
    (Development|PrintString "Player Pawn Found" true true "(R=0.000000,G=0.660000,B=1.000000,A=1.000000)" 1e+01)))

(event OnComponentEndOverlap(Overlap_Volume)
  (params (OverlappedComponent :type Primitive%20Component%20Object%20Reference) (OtherActor :type Actor%20Object%20Reference) (OtherComp :type Primitive%20Component%20Object%20Reference) (OtherBodyIndex :type Integer))
  (bind _asplayer_pawn_athena (Utilities|Casting|CastToPlayerPawn_Athena OtherActor))
  (if (== (Pawn|GetController _asplayer_pawn_athena) (Game|GetPlayerController 0))
    (Rendering|Material|SetScalarParameterValue "/GrillMoon_Countdown/Assets/FX/MPC_GrillMoon.MPC_GrillMoon" "Post_time" (Utilities|Time|GetGameTimeinSeconds) :!class /Script/Engine.KismetMaterialLibrary)
    (Development|PrintString "Player Pawn Found" true true "(R=0.000000,G=0.660000,B=1.000000,A=1.000000)" 1e+01)))

(fn ConstructionScript)
