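# Translucent, two-sided unlit material for the GrillMoon No Fight barrier, a glowing animated wall.
# Panned turbulence, noise-distorted smoke and sparkle layers build the pattern, with time-of-day scaling on the emissive accent.
# An RVT switch selects vertex-color masking or height-above-landscape fades, sampled from landscape and helper-mesh height virtual textures.
# Depth fade, camera-distance sphere masks and time-of-day-dependent player proximity provide additional visibility controls.
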Material /GrillMoon/Assets/TechArt/No_Fight_barrier/M_GrillMoon_No_Fight_Barrier
graph {
 properties {bAutomaticallySetUsageInEditor=false, blendMode=BLEND_Translucent, shadingModel=MSM_Unlit, twoSided=true, usedShadingModels=Unlit}
 nodes [
  n1 = Constant() {r=1}
  n2 = MaterialFunctionCall(Value_Morning=@n1, Value_Day=@n1, Value_Evening=@n1, Value_Night=LinearInterpolate(B=Constant(), Alpha=NamedRerouteUsage()) {constA=1}) {materialFunction=&Scalar_Time_of_Day}
  n3 = VectorParameter() {parameterName=StreaksTileSpeed, channelNames {a=Speed Y, b=Speed X, g=Tile Y, r=Tile X}, defaultValue=<rgba 4 0.3 0.1 0.5>}
  n4 = Multiply(A=TextureCoordinate(), B=ComponentMask(None=@n3) {g=true, r=true})
  n5 = AppendVector(A=ScalarParameter() {parameterName=Offset}, B=Constant())
  n6 = AppendVector(A=@n3.Speed X, B=@n3.Speed X)
  n7 = TextureSample(UVs=Panner(Coordinate=Add(A=@n4, B=@n5), Speed=@n6)) {texture=/Game/Effects/Fort_Effects/Textures/Plasma/T_Turbulence_Stretched}
  n8 = Multiply(
   A=Multiply(A=VectorParameter() {parameterName=Glow_Accent_01, defaultValue=<rgba 0 0 0 1>}.RGB, B=@n7.RGB)
   B=ScalarParameter() {parameterName=recolor_Str}
  )
  MaterialFunctionCall(Emissive=Add(A=VectorParameter() {parameterName=Glow_Color, defaultValue=<rgba 0 18.3434 20 1>}.RGB, B=Multiply(A=@n2.Result, B=@n8))) {materialFunction=&MF_FX_EmissiveAdjustments}
  n9 = NamedRerouteUsage()
  n10 = OneMinus(None=NamedRerouteUsage())
  n11 = TextureSample(UVs=Panner(Coordinate=@n4, Speed=@n6)) {texture=&T_Turbulence_Stretched}
  n12 = Multiply(A=NamedRerouteUsage(), B=Add(A=@n7.RGB, B=@n11.G))
  n13 = ScalarParameter() {defaultValue=1, parameterName=Opacity_Tex_Multi}
  n14 = Multiply(A=@n7.RGB, B=@n13)
  n15 = Multiply(A=@n11.G, B=@n13)
  n16 = VertexColor()
  n17 = OneMinus(None=@n16.R)
  n18 = NamedRerouteUsage()
  n19 = Subtract(
   A=Clamp(None=Add(A=@n14, B=@n15), Max=ScalarParameter() {parameterName=Max_aura}) {maxDefault=0.8}
   B=MaterialFunctionCall!(NoRVT=Saturate(None=@n17), RVT=Saturate(None=@n18)) {materialFunction=/Game/Packages/Fortress_SharedMaterials/Base_Material_Functions/FN_RVT_Switch_Scalar}.Result
  )
  n20 = NamedRerouteUsage()
  n21 = Saturate(None=Multiply(A=@n16.R, B=ScalarParameter() {group=Barrier_Visibility, parameterName=secondary_Multi_non_RVT}))
  n22 = Saturate(None=Multiply(A=OneMinus(None=@n18), B=ScalarParameter() {parameterName=secondary_Multi}))
  n23 = Saturate(None=Add(
    A=Multiply(A=@n19, B=@n20)
    B=MaterialFunctionCall(NoRVT=@n21, RVT=@n22) {materialFunction=&FN_RVT_Switch_Scalar}.Result
   ))
  n24 = Multiply(A=NamedRerouteUsage(), B=OneMinus(None=NamedRerouteUsage()))
  n25 = Add(A=Saturate(None=Multiply(A=@n9, B=@n10)), B=Add(A=Add(A=@n12, B=@n23), B=Saturate(None=@n24)))
  n26 = DepthFade(FadeDistance=ScalarParameter() {defaultValue=100, parameterName=Fade_Distance})
  Multiply(
   A=Add(
    A=MaterialFunctionCall(NoRVT=@n23, RVT=@n25) {materialFunction=&FN_RVT_Switch_Scalar}.Result
    B=LinearInterpolate(Alpha=@n26) {constA=1, constB=0}
   )
   B=NamedRerouteUsage()
  )
  n27 = CameraPositionWS()
  n28 = SphereMask(
   A=WorldPosition() {worldPositionShaderOffset=WPT_ExcludeAllShaderOffsets}.XYZ
   B=@n27
   Radius=ScalarParameter() {group=Barrier_Visibility, parameterName=Visibility Radius Far}
   Hardness=ScalarParameter() {group=Barrier_Visibility, parameterName=Visibility_gradient}
  )
  n29 = SphereMask(
   A=ComponentMask(None=ObjectPositionWS()) {g=true, r=true}
   B=ComponentMask(None=@n27) {g=true, r=true}
   Radius=ScalarParameter() {group=Barrier_Visibility, parameterName=Inside_Radius}
   Hardness=ScalarParameter() {group=Barrier_Visibility, parameterName=Visibility_gradient}
  )
  NamedRerouteDeclaration(None=Saturate(None=Subtract(A=@n28, B=@n29))) {name=Location Fade, nodeColor=<rgba 0.875 1 0 1>}
  n30 = RuntimeVirtualTextureSampleParameter() {
   desc=RVT of the landscape
   group=RVT
   materialType=WorldHeight
   parameterName=RVT-WorldHeightSample
   virtualTexture=/Game/Athena/Apollo/Maps/Landscape/VT/RVT_Apollo_Landscape_Height
  }
  n31 = RuntimeVirtualTextureSample!(UVs, World Position) {
   desc=Helper RVT for meshes that don't natively draw to rvt that we have helper meshes for
   materialType=WorldHeight
   virtualTexture=&/RVT_Grillmoon_Height
  }
  n32 = Subtract(
   A=Add(A=WorldPosition().Z, B=ScalarParameter() {defaultValue=-392.7458, group=RVT, parameterName=height_Offset})
   B=Max(A=@n30.WorldHeight, B=@n31.WorldHeight)
  )
  n33 = ScalarParameter() {defaultValue=2413.3076, group=RVT, parameterName=distance_Blend}
  NamedRerouteDeclaration(None=Power(Base=Divide(A=@n32, B=@n33), Exp=ScalarParameter() {defaultValue=2, group=RVT, parameterName=Distance_Pow})) {name=Height_Landscape, nodeColor=<rgba 1 0 0.9141 1>}
  n34 = VectorParameter() {
   group=Sparkles
   parameterName=Sparkle_02_Pan
   channelNames {a=Y Speed, b=X Speed, g=Y tile, r=X Tile}
   defaultValue=<rgba 2.3879 1.8738 0.15 0.1>
  }
  n35 = Multiply(
   A=TextureCoordinate() {vTiling=-1}
   B=ComponentMask(None=VectorParameter() {group=Sparkles, parameterName=Sparkle_Overall_Tile, defaultValue=<rgba 24 2 0 1>}.RGB) {g=true, r=true}
  )
  n36 = MaterialFunctionCall(Float2=@n35) {materialFunction=&BreakOutFloat2Components}
  n37 = AppendVector(A=@n36.R, B=@n36.R)
  n38 = VectorParameter() {group=Sparkles, parameterName=Spark_Noise_Pan, channelNames {a=Y Speed, b=X Speed, g=Y Tile, r=X tile}, defaultValue=<rgba 2 1 0.1 -0.1>}
  n39 = Panner(Coordinate=Multiply(A=@n35, B=ComponentMask(None=@n38) {g=true, r=true}), Speed=AppendVector(A=@n38.X Speed, B=@n38.X Speed))
  n40 = Multiply(
   A=ComponentMask(None=TextureSample(UVs=@n39) {samplerType=SAMPLERTYPE_LinearColor, texture=/Game/Effects/Fort_Effects/FrontEnd/S33/GateContent/T_FX_Tile_0138_Moon}.RGB) {g=true, r=true}
   B=ScalarParameter() {group=Sparkles, parameterName=Noise_Str}
  ) {constB=0}
  n41 = Panner(
   Coordinate=Multiply(A=ComponentMask(None=@n34) {g=true, r=true}, B=Add(A=@n37, B=@n40) {constB=0})
   Speed=AppendVector(A=@n34.X Speed, B=@n34.X Speed)
  )
  n42 = ScalarParameter() {defaultValue=0.1, group=Sparkles, parameterName=Sparkle_Offset}
  n43 = TextureObject() {texture=&/T_EnvSprite_FinerSparkles}
  n44 = VectorParameter() {group=Sparkles, parameterName=Sparkles_Pan, channelNames {a=Y Speed, b=X Speed, g=Y Tile, r=X Tile}, defaultValue=<rgba 4 3 0.3 0.2>}
  n45 = Multiply(A=Add(A=@n37, B=@n40) {constB=0}, B=ComponentMask(None=@n44) {g=true, r=true})
  n46 = AppendVector(A=@n44.X Speed, B=@n44.X Speed)
  n47 = Add(
   A=TextureSample(UVs=BumpOffset(Coordinate=@n41, Height=@n42), Tex=@n43) {samplerType=SAMPLERTYPE_LinearColor}.R
   B=TextureSample(UVs=Panner(Coordinate=@n45, Speed=@n46), Tex=@n43) {samplerType=SAMPLERTYPE_LinearColor}.R
  )
  n48 = ScalarParameter() {defaultValue=0.6239, group=Sparkles, parameterName=Sparkles_multi}
  NamedRerouteDeclaration(None=Power(Base=Multiply(A=@n47, B=@n48), Exp=ScalarParameter() {defaultValue=1, group=Sparkles, parameterName=Sparkles_Pow})) {name=Sparkles, nodeColor=<rgba 0 0.625 1 1>}
  n49 = ScalarParameter() {defaultValue=600, group=RVT, parameterName=distance_Blend_Sparkles}
  NamedRerouteDeclaration(None=Power(Base=Divide(A=@n32, B=@n49), Exp=ScalarParameter() {defaultValue=1.9677, group=RVT, parameterName=Distance_Pow_Sparkles})) {name=Height_Sparkles, nodeColor=<rgba 1 0 0.7734 1>}
  n50 = VectorParameter() {
   group=Smoke
   parameterName=Smoke Layer 1
   channelNames {a=Tile Speed Y, b=Tile Speed X, g=Tile Y, r=Tile X}
   defaultValue=<rgba 5 4 0.07 -0.04>
  }
  n51 = VectorParameter() {group=Smoke, parameterName=Smoke_OveralTile, channelNames {a="", b="", g=Y Tile, r=X Tile}, defaultValue=<rgba 5 0.4 0 1>}
  n52 = Multiply(A=TextureCoordinate(), B=ComponentMask(None=@n51) {g=true, r=true})
  n53 = MaterialFunctionCall(Float2=@n52) {materialFunction=&BreakOutFloat2Components}
  n54 = Add(A=@n53.R, B=@n53.R)
  n55 = Constant() {r=0.5}
  n56 = Multiply(A=AppendVector(A=@n53.R, B=Subtract(A=@n54, B=@n55)), B=Constant2Vector() {desc="OG -1,-1", g=1, r=1})
  n57 = VectorParameter() {group=Smoke, parameterName=Noise tile/speed, channelNames {a=Y speed, b=X speed, g=Y Tile, r=X tile}, defaultValue=<rgba 2 4 0.1 -0.1>}
  n58 = ComponentMask(None=@n57) {g=true, r=true}
  n59 = Constant() {r=1}
  n60 = TextureSample!(UVs=Panner(Coordinate=Multiply(A=@n58, B=@n52), Speed=AppendVector(A=@n57.X speed, B=@n57.X speed)), Tex, Bias=@n59, Apply View MipBias) {
   mipValueMode=TMVM_MipBias
   samplerSource=SSM_Wrap_WorldGroupSettings
   samplerType=SAMPLERTYPE_LinearColor
   texture=&T_FX_Tile_0138_Moon
  }
  n61 = TextureSample(UVs=@n52, Bias=@n59) {
   mipValueMode=TMVM_MipBias
   samplerSource=SSM_Wrap_WorldGroupSettings
   samplerType=SAMPLERTYPE_LinearColor
   texture=&T_FX_Tile_0138_Moon
  }
  n62 = Multiply(
   A=ComponentMask(None=@n60.RGB) {g=true, r=true}
   B=Multiply(A=ScalarParameter() {defaultValue=0.05, group=Smoke, parameterName=Noise_Overall_Str}, B=@n61.B)
  )
  n63 = Multiply(A=ComponentMask(None=@n50) {g=true, r=true}, B=Add(A=@n56, B=@n62))
  n64 = AppendVector(A=@n50.Tile Speed X, B=@n50.Tile Speed X)
  n65 = TextureObject() {samplerType=SAMPLERTYPE_LinearColor, texture=/Game/Effects/Fort_Effects/FrontEnd/S33/GateContent/Tile_output_8}
  n66 = VectorParameter() {group=Smoke, parameterName=Smoke Layer 2, channelNames {a=Speed Y, b=Speed X, g=Tile Y, r=Tile X}, defaultValue=<rgba 4 3 0.07 0.04>}
  n67 = Multiply(A=ComponentMask(None=@n66) {g=true, r=true}, B=Add(A=@n56, B=@n62))
  n68 = AppendVector(A=@n66.Speed X, B=@n66.Speed X)
  n69 = Multiply(
   A=TextureSample!(UVs=Panner(Coordinate=@n63, Speed=@n64), Tex=@n65, Apply View MipBias) {samplerSource=SSM_Wrap_WorldGroupSettings, samplerType=SAMPLERTYPE_LinearColor}.G
   B=TextureSample(UVs=Panner(Coordinate=@n67, Speed=@n68), Tex=@n65) {samplerSource=SSM_Wrap_WorldGroupSettings, samplerType=SAMPLERTYPE_LinearColor}.G
  )
  n70 = ScalarParameter() {defaultValue=2.5671, group=Smoke, parameterName=PatternMul}
  NamedRerouteDeclaration(None=Power(Base=Multiply(A=@n69, B=@n70), Exp=ScalarParameter() {defaultValue=0.8347, group=Smoke, parameterName=PatternPow})) {name=smoke, nodeColor=<rgba 0.5703 1 0 1>}
  n71 = ScalarParameter() {defaultValue=600, group=RVT, parameterName=distance_Blend_Smoke}
  NamedRerouteDeclaration(None=Power(Base=Divide(A=@n32, B=@n71), Exp=ScalarParameter() {defaultValue=1.9677, group=RVT, parameterName=Distance_Pow_Smoke})) {name=Height_Smoke, nodeColor=<rgba 1 0 0.7734 1>}
  n72 = Divide(A=@n32, B=ScalarParameter() {defaultValue=900, group=RVT, parameterName=distance_Blend_aurora})
  n73 = ScalarParameter() {defaultValue=0.5, group=RVT, parameterName=Distance_Pow_Aurora}
  NamedRerouteDeclaration(None=OneMinus(None=Power(Base=@n72, Exp=@n73))) {name=height_Aurora, nodeColor=<rgba 0.9688 1 0 1>}
  n74 = ScalarParameter() {group=Barrier_Visibility, parameterName=Player_Close_Opacity_Multi_Morning}
  n75 = ScalarParameter() {group=Barrier_Visibility, parameterName=Player_Close_Opacity_Multi_Day}
  n76 = ScalarParameter() {group=Barrier_Visibility, parameterName=Player_Close_Opacity_Multi_Evening}
  n77 = ScalarParameter() {group=Barrier_Visibility, parameterName=Player_Close_Opacity_Multi_Night}
  n78 = WorldPosition() {worldPositionShaderOffset=WPT_ExcludeAllShaderOffsets}
  n79 = CameraPositionWS()
  n80 = ScalarParameter() {group=Barrier_Visibility, parameterName=Player_Close_Distance}
  n81 = ScalarParameter() {group=Barrier_Visibility, parameterName=Player_Close_Gradient}
  NamedRerouteDeclaration(None=LinearInterpolate(
    B=MaterialFunctionCall(Value_Morning=@n74, Value_Day=@n75, Value_Evening=@n76, Value_Night=@n77) {materialFunction=&Scalar_Time_of_Day}.Result
    Alpha=SphereMask(A=@n78.XYZ, B=@n79, Radius=@n80, Hardness=@n81)
   ) {constA=1}) {name=Player proximity, nodeColor=<rgba 0 1 1 1>}
 ]
}
