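# Unlit helper material in the GrillMoon No_Fight_barrier folder that writes world height to a Runtime Virtual Texture.
# The world position Z feeds the WorldHeight input of the Runtime Virtual Texture Output node, its only connected input.
# A Constant3Vector node is present but connects to nothing.
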Material /GrillMoon/Assets/TechArt/No_Fight_barrier/M_GrillMoon_RVT_Helper
graph {
 properties {bAutomaticallySetUsageInEditor=false, shadingModel=MSM_Unlit, usedShadingModels=Unlit}
 nodes [
  Constant3Vector()
  RuntimeVirtualTextureOutput!(BaseColor, Specular, Roughness, Normal, WorldHeight=WorldPosition().Z, Opacity, Mask, Displacement, Mask4)
 ]
}
