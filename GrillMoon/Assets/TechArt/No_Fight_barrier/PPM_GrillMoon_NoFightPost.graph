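# Post-process material for the GrillMoon No Fight barrier, tinting the screen cyan in a timed, screen-centered radial effect.
# An exponential radial gradient masks the tint over scene color, its radius animated by time since the MPC's Post_time, with a testing switch for manual scrubbing.
# Panning fractal and star-flare textures form a noise term that perturbs the animation.
# Several crack, caustic and starburst branches appear unused.
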Material /GrillMoon/Assets/TechArt/No_Fight_barrier/PPM_GrillMoon_NoFightPost
graph {
 properties {materialDomain=MD_PostProcess, usedShadingModels=Default Lit}
 nodes [
  Add()
  n1 = NamedRerouteUsage()
  n2 = ScreenPosition()
  Add(A=@n1, B=@n2.ViewportUV)
  n3 = ScreenPosition()
  n4 = AppendVector(A=ScalarParameter() {group=Noise, parameterName=Noise_SpeedX}, B=ScalarParameter() {group=Noise, parameterName=Noise_SpeedY})
  n5 = Multiply(
   A=TextureSample(UVs=Panner(Coordinate=@n3.ViewportUV, Speed=@n4)) {automaticViewMipBias=false, texture=/Game/Global/Textures/Default/Caustics/T_Fractal_01}.R
   B=ScalarParameter() {defaultValue=0.108, group=Noise, parameterName=noise_Str}
  )
  n6 = MaterialFunctionCall(Vector or UVs=Add(A=TextureCoordinate()) {constB=-0.5}) {materialFunction=&VectorToRadialValue}
  Add(
   A=@n5
   B=Multiply(
    A=TextureSample(UVs=@n6.Radial Coordinates) {automaticViewMipBias=false, texture=/Game/Global/Textures/Default/Caustics/T_Caustics_Softer}.R
    B=ScalarParameter() {defaultValue=0.284, group=Noise, parameterName=noise_Str2}
   )
  )
  n7 = ScalarParameter() {defaultValue=0.492, group=Barrier_Values, parameterName=Max_Value}
  n8 = Add(
   A=StaticSwitchParameter(True=ScalarParameter() {group=Debug/Timing, parameterName=test_time}, False=NamedRerouteUsage()) {group=Debug/Timing, parameterName=Testing?, sortPriority=31}
   B=@n1
  )
  n9 = MaterialFunctionCall(
   UVs=@n2.ViewportUV
   CenterPosition=Constant2Vector() {g=0.5, r=0.5}
   Radius=LinearInterpolate(A=ScalarParameter() {defaultValue=2.3, group=Barrier_Values, parameterName=Radius_Max}, Alpha=Saturate(None=@n8)) {constA=2.3, constB=0}
   Density=ScalarParameter() {defaultValue=2.7152, group=Barrier_Values, parameterName=Density}
  ) {materialFunction=&RadialGradientExponential}
  LinearInterpolate(
   A=ComponentMask(None=SceneTexture() {sceneTextureId=PPI_PostProcessInput0}.Color) {b=true, g=true, r=true}
   B=Multiply(A=Constant3Vector() {constant=<rgba 0 1.8343 2 1>}, B=ScalarParameter() {defaultValue=0.5, group=Barrier_Values, parameterName=str})
   Alpha=Saturate(None=Subtract(A=@n7, B=@n9.RadialGradientExponential))
  )
  MaterialFunctionCall() {materialFunction=&ScreenResolution}
  n10 = Subtract(A=Time(), B=CollectionParameter() {collection=/GrillMoon_Countdown/Assets/FX/MPC_GrillMoon, parameterName=Post_time})
  n11 = ScalarParameter() {defaultValue=1, group=Debug/Timing, parameterName=AnimationDuration}
  n12 = Power(Base=Saturate(None=Divide(A=@n10, B=@n11)), Exp=ScalarParameter() {defaultValue=1, group=Debug/Timing, parameterName=Curve_Amount})
  NamedRerouteDeclaration(None=Clamp(None=OneMinus(None=@n12))) {name=Time_Anim, nodeColor=<rgba 0.2891 0 1 1>}
  n13 = TextureSample(UVs=MaterialFunctionCall(Texture Scale=ScalarParameter() {defaultValue=1.4385, group=Noise, parameterName=Radial_Star_Scale}) {materialFunction=&ScaleUVsByCenter}.UVs) {samplerType=SAMPLERTYPE_LinearColor, texture=/Game/Effects/Fort_Effects/GameplayPulginsAssets/DiamondSkin/Textures/T_DiamondSkin_Flare}
  n14 = Multiply(A=Saturate(None=@n13.R), B=ScalarParameter() {group=Noise, parameterName=star_str})
  NamedRerouteDeclaration(None=Saturate(None=Add(A=@n5, B=@n14))) {name=noise, nodeColor=<rgba 0 1 0.1562 1>}
  StaticBool() {value=true}
  Subtract(
   A=@n13.R
   B=MaterialFunctionCall(
    UVs=ScreenPosition().ViewportUV
    CenterPosition=Constant2Vector() {g=0.5, r=0.5}
    Radius=ScalarParameter() {defaultValue=0.2, parameterName=Starburst Radius}
    Density=ScalarParameter() {defaultValue=4, parameterName=Starburst density}
   ) {materialFunction=&RadialGradientExponential}.RadialGradientExponential
  )
  n15 = Multiply(A=ScreenPosition().ViewportUV, B=ScalarParameter() {defaultValue=1, parameterName=Scale})
  TextureSample(UVs=@n15) {texture=/Game/Athena/Environments/Meteor/Textures/Meteor_Cracks_M}
  TextureSample(UVs=@n15) {texture=/Game/Effects/Fort_Effects/Gameplay/SmokeDustExplosions/Textures/T_Decal_Cracked}
  TextureSample() {texture=/Game/Effects/Fort_Effects/Textures/Electric/T_Vert_Electric_Cracks}
  TextureSample(UVs=@n6.Radial Coordinates) {automaticViewMipBias=false, texture=/Game/Global/Textures/Default/Caustics/T_Turbulence_Seamless}
 ]
}
